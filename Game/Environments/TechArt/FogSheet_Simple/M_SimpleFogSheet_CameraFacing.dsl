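; Camera-facing fog sheet Blueprint that configures a plane with a simple fog material; its event graph is empty.
; The construction script runs a shadow-quality check for plane visibility, where only the Low tier's display flag can hide it.
; It then creates a dynamic material instance and feeds it brightness, depth and near fade, panning and edge falloff.
; Color comes from either four time-of-day colors or a single fog color.
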(event EventBeginPlay)

(event Collision|EventActorBeginOverlap
  (params (OtherActor :type Actor%20Object%20Reference)))

(event EventTick
  (params (DeltaSeconds :type Float%20%28single-precision%29)))

(fn ConstructionScript
  (bind _mid (Variables|Default|GetMID))
  (Utilities|FlowControl|Sequence
    (:then_0
      ;[1 1 1] Scalibility
      (Utilities|ShadowQualityCheckVFX
        (:Low
          (if (not (Variables|Scalability|GetDisplay_Low))
            (Rendering|SetVisibility (Variables|Default|GetPlane) false false)))
        (:Medium
          (if (Variables|Scalability|GetDisplay_Medium)
            _
            (else
              _)))
        (:High
          (if (Variables|Scalability|GetDisplay_High)
            _
            (else
              _)))
        (:Epic
          (if (Variables|Scalability|GetDisplay_Epic)
            _
            (else
              _))))
      (Rendering|SetVisibility (Variables|Default|GetPlane) true false))
    (:then_1
      (bind _returnvalue (Rendering|Material|CreateDynamicMaterialInstance (Variables|Default|GetPlane) 0 "/Game/Environments/TechArt/FogSheet_Simple/M_SimpleFogSheet.M_SimpleFogSheet" "None" :!class /Script/Engine.PrimitiveComponent))
      (Variables|Default|SetMID _returnvalue)
      (if (Variables|TODM|GetUse%20TOD%20Color)
        (Rendering|Material|SetVectorParameterValue _mid "Color Morning" (Variables|TODM|GetColor%20Morning) :!class /Script/Engine.MaterialInstanceDynamic)
        (Rendering|Material|SetVectorParameterValue _mid "Color" (Variables|TODM|GetColor%20Day) :!class /Script/Engine.MaterialInstanceDynamic)
        (Rendering|Material|SetVectorParameterValue _mid "Color Evening" (Variables|TODM|GetColor%20Evening) :!class /Script/Engine.MaterialInstanceDynamic)
        (Rendering|Material|SetVectorParameterValue _mid "Color Night" (Variables|TODM|GetColor%20Night) :!class /Script/Engine.MaterialInstanceDynamic)
        (else
          (Rendering|Material|SetVectorParameterValue _mid "Color" (Variables|Default|GetFog%20Color) :!class /Script/Engine.MaterialInstanceDynamic)))
      (Rendering|Material|SetScalarParameterValue _mid "FogBrightness" (Variables|Default|GetFog%20Brightness) :!class /Script/Engine.MaterialInstanceDynamic)
      (Rendering|Material|SetScalarParameterValue _mid "DepthFadeDistance" (Variables|Default|GetDepth%20Fade%20Distance) :!class /Script/Engine.MaterialInstanceDynamic)
      (Rendering|Material|SetScalarParameterValue _mid "NearFadeOffset" (Variables|Default|GetNear%20Fade%20Offset) :!class /Script/Engine.MaterialInstanceDynamic)
      (Rendering|Material|SetScalarParameterValue _mid "NearFadeBlend" (Variables|Default|GetNear%20Fade%20Blend) :!class /Script/Engine.MaterialInstanceDynamic)
      (Rendering|Material|SetScalarParameterValue _mid "PanningSpeed" (Variables|Default|GetPanning%20Speed) :!class /Script/Engine.MaterialInstanceDynamic)
      (Rendering|Material|SetScalarParameterValue _mid "FadePower" (Variables|Default|GetFalloffEdges) :!class /Script/Engine.MaterialInstanceDynamic))))
